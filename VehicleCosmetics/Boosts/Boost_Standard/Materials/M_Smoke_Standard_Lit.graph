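# Lit translucent smoke material for Niagara sprites in the standard vehicle boost effect, using volumetric per-vertex directional lighting.
# Dynamic-parameter animation phase drives 4x4 flipbooks for depth, alpha, emissive and normals.
# Normals blend flipbook detail with a spherical normal and optional directional bias.
# Per-particle dynamic parameters scale HDR emissive and control scrolled, distorted noise that erodes emissive and opacity, with particle alpha fading the smoke.
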Material /VehicleCosmetics/Boosts/Boost_Standard/Materials/M_Smoke_Standard_Lit
graph {
 properties {
  bUsedWithNanite=true
  bUsedWithNiagaraSprites=true
  blendMode=BLEND_Translucent
  pixelDepthOffsetMode=PDOM_Legacy
  translucencyLightingMode=TLM_VolumetricPerVertexDirectional
  translucentSelfShadowDensityScale=1
  usedShadingModels=Default Lit
  translucentMultipleScatteringExtinction=<rgba 1 1 1 1>
 }
 nodes [
  MaterialFunctionCall(Emissive=NamedRerouteUsage()) {materialFunction=&MF_FX_EmissiveAdjustments}
  n1 = DynamicParameter() {paramNames [Animation, Unused, Emissive, NormalMultiplier]}
  n2 = Constant() {r=4}
  n3 = TextureObject() {texture=/Game/Effects/Niagara/MasterMaterials/Smoke/Textures/NEW/RT4x4_1_Tex}
  NamedRerouteDeclaration(None=ComponentMask(None=MaterialFunctionCall("Animation  Phase (0-1)"=@n1.Animation, Number of Rows=@n2, Number of Columns=@n2, Texture=@n3) {materialFunction=&FlipBook}.Result) {g=true}) {name=Depth_Flipbook, nodeColor=<rgba 0 0.8594 1 1>}
  n4 = TextureObject() {texture=/Game/Effects/Niagara/MasterMaterials/Smoke/Textures/NEW/RT4x4_4_Tex}
  NamedRerouteDeclaration(None=ComponentMask(None=MaterialFunctionCall("Animation  Phase (0-1)"=@n1.Animation, Number of Rows=@n2, Number of Columns=@n2, Texture=@n4) {materialFunction=&FlipBook}.Result) {g=true}) {name=Alpha_Flipbook, nodeColor=<rgba 1 0 0.3516 1>}
  n5 = ComponentMask(None=NamedRerouteUsage()) {g=true, r=true}
  n6 = DynamicParameter() {
   parameterIndex=2
   defaultValue=<rgba 0 0 0 1>
   paramNames [Smoke - DissolveLerpNoise, Smoke - DissolveNoiseOffsetX, Smoke - DissolveNoiseErosionPower, Smoke - DissolveNoiseScale]
  }
  n7 = Constant()
  n8 = ScalarParameter() {group=Normal, parameterName="DissolveDetail  Random Y Offset Value", sliderMax=2}
  n9 = Add(A=Add(A=TextureCoordinate(), B=@n6.Smoke - DissolveNoiseOffsetX), B=Multiply(A=ParticleRandom(), B=AppendVector(A=@n7, B=@n8)))
  n10 = TextureSampleParameter2D(UVs=Add(A=Multiply(A=@n5) {constB=-0.2}, B=Multiply(A=@n9, B=@n6.Smoke - DissolveNoiseScale) {desc=Scale})) {parameterName=DissolveDetailTexture, samplerType=SAMPLERTYPE_LinearColor, texture=/Game/Global/Textures/Default/Noise/T_Noise01_Inverse}
  n11 = DynamicParameter() {
   parameterIndex=2
   defaultValue=<rgba 0 0 0 1>
   paramNames [Smoke - DissolveLerpNoise, Smoke - DissolveNoiseOffsetX, Smoke - DissolveNoiseErosionPower, Smoke - DissolveNoiseScale]
  }
  NamedRerouteDeclaration(None=Clamp(None=Power(Base=@n10.G, Exp=@n11.Smoke - DissolveNoiseErosionPower))) {name=DissolveDetail, nodeColor=<rgba 0.3828 0 1 1>}
  n12 = Subtract(A=NamedRerouteUsage(), B=Constant() {r=-1})
  n13 = ScalarParameter() {parameterName=Contrast}
  n14 = Multiply(
   A=Multiply(
    A=MaterialFunctionCall(In=@n12, Contrast=@n13) {materialFunction=&CheapContrast}.Result
    B=NamedRerouteUsage()
   )
   B=ParticleColor().A
  )
  n15 = NamedRerouteUsage()
  n16 = DynamicParameter() {
   parameterIndex=2
   defaultValue=<rgba 0 0 0 1>
   paramNames [Smoke - DissolveLerpNoise, Smoke - DissolveNoiseOffsetX, Smoke - DissolveNoiseErosionPower, Smoke - DissolveNoiseScale]
  }
  NamedRerouteDeclaration(None=LinearInterpolate(A=@n14, B=Multiply(A=@n14, B=@n15), Alpha=Clamp(None=@n16.Smoke - DissolveLerpNoise))) {name=Final Alpha, nodeColor=<rgba 1 0 0.2578 1>}
  NamedRerouteDeclaration(None=Multiply(A=VectorParameter() {parameterName=Color, defaultValue=<rgba 0.2448 0.2342 0.2318 1>}.RGB) {constB=0.4}) {name=Final_Base Color, nodeColor=<rgba 1 0.7266 0 1>}
  NamedRerouteDeclaration(None=Multiply(A=ParticleColor().RGB, B=NamedRerouteUsage())) {name=Final_Emissive, nodeColor=<rgba 1 0.8672 0 1>}
  n17 = TextureObject() {texture=/Game/Effects/Niagara/MasterMaterials/Smoke/Textures/NEW/RT4x4_5_Tex}
  NamedRerouteDeclaration(None=ComponentMask(None=MaterialFunctionCall("Animation  Phase (0-1)"=@n1.Animation, Number of Rows=@n2, Number of Columns=@n2, Texture=@n17) {materialFunction=&FlipBook}.Result) {g=true}) {name=Emissive_Flipbook, nodeColor=<rgba 0 1 0.6953 1>}
  n18 = TextureObjectParameter() {parameterName=NormalTexture, samplerType=SAMPLERTYPE_Normal, texture=/Game/Effects/Niagara/MasterMaterials/Smoke/Textures/NEW/RT4x4_3_Tex}
  NamedRerouteDeclaration(None=ComponentMask(None=MaterialFunctionCall("Animation  Phase (0-1)"=@n1.Animation, Number of Rows=@n2, Number of Columns=@n2, Texture=@n18) {materialFunction=&FlipBook}.Result) {b=true, g=true, r=true}) {name=Normal_Flipbook, nodeColor=<rgba 0.9453 0 1 1>}
  n19 = Constant3Vector() {constant=<rgba 0 0 1 0>}
  n20 = NamedRerouteDeclaration(None=TextureSample() {samplerType=SAMPLERTYPE_Normal, texture=/Engine/Functions/Engine_MaterialFunctions02/ExampleContent/Textures/SphereRenderNormalsMap}.RGB) {name=SphereNormal_Texture, nodeColor=<rgba 1 0.3281 0 1>}
  n21 = Multiply(
   A=ScalarParameter() {defaultValue=1, group=Normal, parameterName=SphereNormal}
   B=DynamicParameter() {paramNames [Animation, Unused, Emissive, NormalMultiplier]}.NormalMultiplier
  )
  NamedRerouteDeclaration(None=Normalize(VectorInput=LinearInterpolate(A=@n19, B=@n20, Alpha=@n21))) {name=SphereNormal, nodeColor=<rgba 1 0 0.2109 1>}
  n22 = NamedRerouteUsage()
  n23 = ComponentMask(None=VectorParameter() {group=Normal, parameterName=NormalBias Direction, defaultValue=<rgba 0.5 0.5 0 0>}.RGB) {b=true, g=true, r=true}
  n24 = ScalarParameter() {defaultValue=1, desc=Toggle the Bias direction from Tangent to World Space, group=Normal, parameterName=NormalBias WorldSpace}
  n25 = Add(
   A=@n22
   B=LinearInterpolate(A=@n23, B=Transform(None=@n23) {transformSourceType=TRANSFORMSOURCE_World, transformType=TRANSFORM_Tangent}, Alpha=@n24)
  )
  n26 = ComponentMask(None=VectorParameter() {group=Normal, parameterName=NormalBias Multiplier, defaultValue=<rgba 2 2 1 0>}.RGB) {b=true, g=true, r=true}
  NamedRerouteDeclaration(None=StaticSwitchParameter(True=@n22, False=Multiply(A=@n25, B=@n26)) {group=Normal, parameterName=DontUseNormalBias}) {name=BiasedNormal, nodeColor=<rgba 0.6875 1 0 1>}
  n27 = NamedRerouteUsage()
  n28 = NamedRerouteUsage()
  n29 = ScalarParameter() {defaultValue=1, group=Normal, parameterName=NormalTexture}
  NamedRerouteDeclaration(None=Normalize(VectorInput=LinearInterpolate(A=@n27, B=@n28, Alpha=@n29))) {name=Final Normal, nodeColor=<rgba 1 0 0.75 1>}
  n30 = NamedRerouteUsage()
  n31 = Multiply(A=NamedRerouteUsage(), B=@n30)
  n32 = MaterialFunctionCall(
   LDR Input=StaticSwitchParameter(True=Clamp(None=@n31), False=@n30) {parameterName=Multiply HDR Emissive by Base Color}
   HDR Intensity Multiplier=Multiply(
    A=ScalarParameter() {defaultValue=100, parameterName=HDR_Intensity}
    B=DynamicParameter() {paramNames [Animation, Unused, Emissive, NormalMultiplier]}.Emissive
   )
  ) {materialFunction=&DeriveHDRfromLDR}
  n33 = Multiply(A=ComponentMask(None=NamedRerouteUsage()) {g=true, r=true}) {constB=-0.2}
  n34 = DynamicParameter() {
   parameterIndex=1
   defaultValue=<rgba 0 0 0 1>
   paramNames [Emissive - LerpNoise, Emissive - NoiseOffsetX, Emissive - NoiseErosionPower, Emissive - NoiseScale]
  }
  n35 = Multiply(A=Add(A=TextureCoordinate(), B=@n34.Emissive - NoiseOffsetX), B=@n34.Emissive - NoiseScale) {desc=Scale}
  n36 = DynamicParameter() {
   parameterIndex=1
   defaultValue=<rgba 0 0 0 1>
   paramNames [Emissive - LerpNoise, Emissive - NoiseOffsetX, Emissive - NoiseErosionPower, Emissive - NoiseScale]
  }
  n37 = Power(
   Base=TextureSampleParameter2D(UVs=Add(A=@n33, B=@n35)) {parameterName=DetailTexture, texture=/Game/Global/Textures/Default/Tileable_Patterns/T_Blowing_Smoke_Pattern2_Tile_DISP}.G
   Exp=@n36.Emissive - NoiseErosionPower
  )
  n38 = DynamicParameter() {
   parameterIndex=1
   defaultValue=<rgba 0 0 0 1>
   paramNames [Emissive - LerpNoise, Emissive - NoiseOffsetX, Emissive - NoiseErosionPower, Emissive - NoiseScale]
  }
  NamedRerouteDeclaration(None=LinearInterpolate(A=@n32.RemappedRange, B=Multiply(A=@n32.RemappedRange, B=@n37), Alpha=Clamp(None=@n38.Emissive - LerpNoise))) {name=HDR Emissive, nodeColor=<rgba 1 0 0.3047 1>}
  NamedRerouteUsage()
  NamedRerouteUsage()
  NamedRerouteUsage()
 ]
}
